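# Small custom tooltip widget with an optional icon and header text, a dotted rule, and a body that hosts a supplied widget.
# On Construct it adds ContentWidget to a named slot, then collapses the icon, header, rule or body when their data is missing.
# Body spacers follow icon and header visibility to indent the content.
# Getter functions expose header text and icon brush, while Get Body Text returns nothing.
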WidgetBlueprint /Game/UI/Tooltips/Tooltip-Custom-S
graphs [
 EventGraph {nodes [
   Event Construct()
   > n1 = Sequence!(execute) -> (then_0, then_1, then_2:Exec, then_3:Exec, then_4:Exec, then_5:Exec, then_6:Exec, then_7:Exec)
   > AddChild(
    self=Get Content Slot() -> (Content Slot:Named Slot Object Reference)
    Content=Get ContentWidget() -> (ContentWidget:User Widget Object Reference)
   )
   n2 = IsValid(Object=Get Icon Brush() -> (Icon Brush:Slate Brush Asset Object Reference))
   n3 = Select(Option 0=Collapsed, Option 1=Visible, Index=@n2)
   SetVisibility(execute=@n1.then_1, self=Get Icon-Header() -> (Icon-Header:Image Object Reference), InVisibility=@n3)
   SetVisibility(execute=@n1.then_2, self=Get Spacer-Icon() -> (Spacer-Icon:Spacer Object Reference), InVisibility=@n3)
   n4 = NOT Boolean(A=TextIsEmpty(InText=Get HeaderText() -> (HeaderText:Text)))
   n5 = Select(Option 0=Collapsed, Option 1=Visible, Index=@n4)
   SetVisibility(execute=@n1.then_3, self=Get Text-Header() -> (Text-Header:Text Object Reference), InVisibility=@n5)
   SetVisibility(execute=@n1.then_4, self=Get Spacer-Header() -> (Spacer-Header:Spacer Object Reference), InVisibility=@n5)
   n6 = Select(Option 0=Collapsed, Option 1=Visible, Index=OR Boolean(A=@n2, B=@n4))
   SetVisibility(execute=@n1.then_5, self=Get Header HBox() -> (Header HBox:Horizontal Box Object Reference), InVisibility=@n6)
   SetVisibility(execute=@n1.then_6, self="Get <hr>"() -> ("<hr>":Image Object Reference), InVisibility=@n6)
   n7 = Get ContentWidget()
   SetVisibility(
    execute=@n1.then_7
    self=Get Body HBox() -> (Body HBox:Horizontal Box Object Reference)
    InVisibility=Select(Option 0=Collapsed, Option 1=Visible, Index=IsValid(Object=@n7))
   )
  ]}
 Get Body Text {nodes [
   Get Body Text() {is=title K2Node_FunctionEntry}
   > Return Node!(execute, ReturnValue:Text)
  ]}
 Get Header Text {nodes [
   Get Header Text() {is=title K2Node_FunctionEntry}
   > Return Node(ReturnValue=Get HeaderText())
  ]}
 Get Icon Brush {nodes [
   Get Icon Brush() {is=title K2Node_FunctionEntry}
   > Return Node(ReturnValue:Slate Brush Structure=MakeBrushFromAsset!(BrushAsset:Slate Brush Asset Object Reference=Get Icon Brush() {is=title K2Node_VariableGet}) -> (ReturnValue:Slate Brush Structure))
  ]}
]
variables [
 ContentWidget {type=UserWidget, default=null}
 HeaderText {type=string, default=""}
 Icon Brush {type=SlateBrushAsset, default=null}
]
widgets {
 info {inheritedWidgetCount=0, parentClass=&UserWidget, rootWidgetClass=&SizeBox, widgetCount=13}
 widgets [
  SizeBox_145 {
   widgetClassPath=&SizeBox
   widget=SizeBox_145
   properties {bOverride_MinDesiredHeight=true, bOverride_MinDesiredWidth=true, maxDesiredWidth=512, minDesiredHeight=32, minDesiredWidth=128}
  }
  Overlay_0 {widgetClassPath=&Overlay, parent=SizeBox_145, widget=Overlay_0, slot {class=SizeBoxSlot}}
  CommonBorder_268 {
   widgetClassPath=&CommonBorder
   parent=Overlay_0
   widget=CommonBorder_268
   properties {
    style=/FortUILibrary/Styles/Border/Border-SolidBG-ToolTipShadow.Border-SolidBG-ToolTipShadow_C
    background {
     bIsDynamicallyLoaded=false
     drawAs=Box
     imageType=NoImage
     mirroring=NoMirror
     resourceName=None
     resourceObject=/FortUILibrary/Styles/Border/Materials/M-Border-ToolTip
     tiling=NoTile
     imageSize=<32 32>
     margin=<ltrb 0.5 0.5 0.5 0.5>
     outlineSettings {
      bUseBrushTransparency=false
      roundingType=HalfHeightRadius
      width=0
      color {colorUseRule=UseColor_Specified, specifiedColor=<rgba 0 0 0 0>}
      cornerRadii=<0 0 0 0>
     }
     tintColor {colorUseRule=UseColor_Specified, specifiedColor=<rgba 1 1 1 1>}
     uVRegion {bIsValid=false, max=<0 0>, min=<0 0>}
    }
    renderTransform {angle=0, scale=<1 1>, shear=<0 0>, translation=<6 8>}
   }
   slot {class=OverlaySlot, properties {horizontalAlignment=HAlign_Fill, verticalAlignment=VAlign_Fill}}
  }
  CommonBorder_0 {
   widgetClassPath=&CommonBorder
   parent=Overlay_0
   widget=CommonBorder_0
   properties {style=/FortUILibrary/Styles/Border/Border-SolidBG-ToolTip.Border-SolidBG-ToolTip_C, padding=<ltrb 6 6 6 6>}
   slot {class=OverlaySlot}
  }
  VerticalBox_509 {
   widgetClassPath=&VerticalBox
   parent=CommonBorder_0
   widget=VerticalBox_509
   slot {class=BorderSlot, properties {padding=<ltrb 6 6 6 6>}}
  }
  Header HBox {widgetClassPath=&HorizontalBox, bIsVariable=true, parent=VerticalBox_509, widget=Header HBox, slot {class=VerticalBoxSlot}}
  Icon-Header {
   widgetClassPath=&Image
   bIsVariable=true
   parent=Header HBox
   widget=Icon-Header
   properties {brush {
     bIsDynamicallyLoaded=false
     drawAs=Image
     imageType=NoImage
     mirroring=NoMirror
     resourceName=None
     resourceObject=/Game/UI/Icons/Icon-BasicFrame-S
     tiling=NoTile
     imageSize=<24 24>
     margin=<ltrb 0 0 0 0>
     outlineSettings {
      bUseBrushTransparency=false
      roundingType=HalfHeightRadius
      width=0
      color {colorUseRule=UseColor_Specified, specifiedColor=<rgba 0 0 0 0>}
      cornerRadii=<0 0 0 0>
     }
     tintColor {colorUseRule=UseColor_Specified, specifiedColor=<rgba 1 1 1 1>}
     uVRegion {bIsValid=false, max=<0 0>, min=<0 0>}
    }}
   slot {class=HorizontalBoxSlot, properties {horizontalAlignment=HAlign_Left, verticalAlignment=VAlign_Top, padding=<ltrb 0 0 4 4>}}
  }
  Text-Header {
   widgetClassPath=&TextBlock
   bIsVariable=true
   parent=Header HBox
   widget=Text-Header
   properties {
    text=Header Text
    font {
     bForceMonospaced=false
     bMaterialIsStencil=false
     fontMaterial=null
     fontObject=&Burbank
     letterSpacing=0
     monospacedWidth=1
     size=18
     skewAmount=0
     typefaceFontName=Burbank-Black
     outlineSettings {
      bApplyOutlineToDropShadows=false
      bMiteredCorners=false
      bSeparateFillAlpha=false
      outlineMaterial=null
      outlineSize=0
      outlineColor=<rgba 0 0 0 1>
     }
    }
    shadowColorAndOpacity=<rgba 0 0 0 1>
   }
   slot {class=HorizontalBoxSlot, properties {padding=<ltrb 0 1 0 -1>}}
  }
  "<hr>" {
   widgetClassPath=&Image
   bIsVariable=true
   parent=VerticalBox_509
   widget="<hr>"
   properties {brush {
     bIsDynamicallyLoaded=false
     drawAs=Image
     imageType=NoImage
     mirroring=NoMirror
     resourceName=None
     resourceObject=/Game/UI/Foundation/Textures/Rule/Rule-Dotted-8
     tiling=Horizontal
     imageSize=<6 6>
     margin=<ltrb 0 0 0 0>
     outlineSettings {
      bUseBrushTransparency=false
      roundingType=HalfHeightRadius
      width=0
      color {colorUseRule=UseColor_Specified, specifiedColor=<rgba 0 0 0 0>}
      cornerRadii=<0 0 0 0>
     }
     tintColor {colorUseRule=UseColor_Specified, specifiedColor=<rgba 0.4452 0.5972 1 1>}
     uVRegion {bIsValid=false, max=<0 0>, min=<0 0>}
    }}
   slot {class=VerticalBoxSlot, properties {padding=<ltrb 0 0 0 4>}}
  }
  Body HBox {
   widgetClassPath=&HorizontalBox
   bIsVariable=true
   parent=VerticalBox_509
   widget=Body HBox
   slot {class=VerticalBoxSlot, properties {padding=<ltrb 0 4 0 0>}}
  }
  Spacer-Header {
   widgetClassPath=&Spacer
   bIsVariable=true
   parent=Body HBox
   widget=Spacer-Header
   properties {size=<8 1>}
   slot {class=HorizontalBoxSlot}
  }
  Spacer-Icon {
   widgetClassPath=&Spacer
   bIsVariable=true
   parent=Body HBox
   widget=Spacer-Icon
   properties {size=<24 1>}
   slot {class=HorizontalBoxSlot}
  }
  Content Slot {
   widgetClassPath=&NamedSlot
   bIsVariable=true
   parent=Body HBox
   widget=Content Slot
   properties {bExposeOnInstanceOnly=true}
   slot {class=HorizontalBoxSlot, properties {size {sizeRule=Fill, value=1}}}
  }
 ]
}
