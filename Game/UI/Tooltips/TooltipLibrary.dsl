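; TooltipLibrary is a UI function library that builds tooltip widgets.
; Creation functions spawn basic, custom-content and XP bar tooltip widgets for a player controller, passing in text and, where applicable, styling, icon or inner widget.
; The item and compare-item functions are legacy stubs that only print a warning and return no widget.
; SetEnabledAndTooltipText toggles a widget's enabled state and attaches a basic tooltip, or clears it when the text is empty.
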(fn CreateBasicTooltip
  (params (Owning%20Player :type Player%20Controller%20Object%20Reference) (Body%20Text :type Text) (Header%20Text :type Text) (Icon%20Brush :type Slate%20Brush%20Asset%20Object%20Reference) (Header%20Style :type Common%20Text%20Style%20Class%20Reference) (Body%20Style :type Common%20Text%20Style%20Class%20Reference))
  (bind _returnvalue (UserInterface|CreateWidget 0 "/Game/UI/Tooltips/Tooltip-Basic-S.Tooltip-Basic-S_C" Owning%20Player Body%20Text Header%20Text Icon%20Brush Header%20Style Body%20Style))
  (return _returnvalue))

(fn CreateCompareItemTooltip
  (params (Owning%20Player :type Player%20Controller%20Object%20Reference) (Header%20Text :type Text) (Header%20Icon :type Slate%20Brush%20Asset%20Object%20Reference) (Item :type Fort%20Item%20Object%20Reference) (ComparedItem :type Fort%20Item%20Object%20Reference))
  (Development|PrintText 0 (Utilities|Text|FormatText "NSLOCTEXT(\"[A93D42A74C5182F207B53B80C58EFDF8]\", \"6D95A6B849411FB7133689A3B5C14C79\", \"THIS IS A LEGACY TOOLTIP FUNCTION (HEADER TEXT \\\"{0}\\\", OBJECT {1})\")" Header%20Text (Utilities|Text|ToText(String) (Utilities|GetObjectName Item))))
  (return 0))

(fn CreateCustomTooltip
  (params (Owning%20Player :type Player%20Controller%20Object%20Reference) (Header%20Text :type Text) (Icon%20Brush :type Slate%20Brush%20Asset%20Object%20Reference) (Widget :type User%20Widget%20Object%20Reference))
  (bind _returnvalue (UserInterface|CreateWidget 0 "/Game/UI/Tooltips/Tooltip-Custom-S.Tooltip-Custom-S_C" Owning%20Player Header%20Text Icon%20Brush Widget))
  (return _returnvalue))

(fn CreateItemTooltip
  (params (Owning%20Player :type Player%20Controller%20Object%20Reference) (Header%20Text :type Text) (Header%20Icon :type Slate%20Brush%20Asset%20Object%20Reference) (Item :type Fort%20Item%20Object%20Reference) (OverrideQuantity :type Integer :default -1))
  (Development|PrintText Owning%20Player (Utilities|Text|FormatText "NSLOCTEXT(\"[A93D42A74C5182F207B53B80C58EFDF8]\", \"B5684C194CCCDCA6F161FC841CED5D72\", \"THIS IS A LEGACY TOOLTIP FUNCTION (HEADER TEXT \\\"{0}\\\", OBJECT {1})\")" Header%20Text (Utilities|Text|ToText(String) (Utilities|GetObjectName Item))))
  (return 0))

(fn CreateXPTooltip
  (params (Owner :type Player%20Controller%20Object%20Reference) (Header :type Text) (CommanderXP :type Text) (BoostXP :type Text) (RestXP :type Text) (Body :type Text))
  (bind _returnvalue (UserInterface|CreateWidget 0 "/Game/UI/Frontend/XpBar/XpBarToolTip.XpBarToolTip_C" Owner Header CommanderXP BoostXP RestXP Body))
  (return _returnvalue))

(fn SetEnabledAndTooltipText
  (params (Widget :type Widget%20Object%20Reference) (Enabled :type Boolean) (Tooltip%20Text :type Text))
  (Widget|SetIsEnabled Widget Enabled)
  (if (Utilities|Text|TextIsEmpty Tooltip%20Text)
    (Widget|SetToolTip Widget 0)
    (else
      (bind _output (CallFunction|CreateBasicTooltip 0 Tooltip%20Text "" 0 0 0 Widget))
      (Widget|SetToolTip Widget _output))))
