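# Popup tooltip widget showing an uppercase main line and a rich-text second line in an angled, shadowed box with a quest-style pointer arrow.
# PreConstruct applies first-line text and configurable arrow alignment, padding and transform, and collapses the second line when empty.
# ShowNotification and HideNotification play opacity fades guarded by IsShowing; Construct jumps to the hidden end state.
# OnInitialized scales the widget up on mobile.
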WidgetBlueprint /Game/UI/Tooltips/WBP_PopupTooltip
animations [
 ShowAnimation {
  binding_count=1
  length_seconds=0.15
  object=&self:ShowAnimation
  sequence {bindings [DiscoveryTooltip_C_18 {
     id=#1
     possessed_class=WBP_PopupTooltip_C
     tracks [{
       type=MovieSceneFloatTrack
       display_name=Render Opacity
       sections [{type=MovieSceneFloatSection, channels [None {type=MovieSceneScriptingFloatChannel, keys [[0,0],[3,1]]}]}]
      }]
    }]}
 }
 HideAnimation {
  binding_count=1
  length_seconds=0.15
  object=&self:HideAnimation
  sequence {bindings [DiscoveryTooltip_C_19 {
     id=#2
     possessed_class=WBP_PopupTooltip_C
     tracks [{
       type=MovieSceneFloatTrack
       display_name=Render Opacity
       sections [{type=MovieSceneFloatSection, channels [None {type=MovieSceneScriptingFloatChannel, keys [[0,1],[3,0]]}]}]
      }]
    }]}
 }
]
graphs [
 EventGraph {nodes [
   ShowNotification() {is=title K2Node_CustomEvent}
   > n1 = Branch(Condition=Get IsShowing() -> (IsShowing:bool))
   Set IsShowing(execute=@n1.else, IsShowing:bool=true) -> (Output_Get:bool)
   > PlayAnimation(InAnimation=Get ShowAnimation() -> (ShowAnimation:Widget Animation Object Reference))
   HideNotification() {is=title K2Node_CustomEvent}
   > Branch(Condition=Get IsShowing())
   > Set IsShowing(IsShowing=false)
   > PlayAnimation(InAnimation=Get HideAnimation() -> (HideAnimation:Widget Animation Object Reference))
   Event OnInitialized()
   > Branch(Condition=IsMobileGame())
   > SetRenderScale!(execute, self:Widget Object Reference, Scale:Vector 2D Structure=(X=1.26,Y=1.26)) -> (then)
   Event PreConstruct()
   > SetFirstLineText!(execute, self, FirstLineText:Text=Get FirstLineText() -> (FirstLineText:Text)) -> (then)
   n2 = SlotAsOverlaySlot(Widget=Get Image_Arrow() -> (Image_Arrow:Image Object Reference))
   > SetPadding!(execute, self:Overlay Slot Object Reference=@n2, InPadding:Margin Structure=Get ArrowPadding() -> (ArrowPadding:Margin Structure)) -> (then)
   > SetHorizontalAlignment(self=@n2, InHorizontalAlignment=Get ArrowHorizontalAlignment() -> (ArrowHorizontalAlignment:EHorizontalAlignment Enum=HAlign_Fill))
   > SetVerticalAlignment(self=@n2, InVerticalAlignment=Get ArrowVerticalAlignment() -> (ArrowVerticalAlignment:EVerticalAlignment Enum=VAlign_Fill))
   > SetRenderTransform!(
    execute
    self:Widget Object Reference=Get Image_Arrow()
    InTransform:Widget Transform Structure=Get ArrowTransform() -> (ArrowTransform:Widget Transform Structure)
   ) -> (then)
   > n3 = Branch(Condition=TextIsEmpty(InText=Get SecondLineRichText() -> (SecondLineRichText:Text)))
   n4 = Get RichText_SecondaryLine() -> (RichText_SecondaryLine:Common Rich Text Block Object Reference)
   > SetVisibility(self=@n4, InVisibility=Collapsed)
   SetText(execute=@n3.else, self=@n4, InText=Get SecondLineRichText())
   Event Construct()
   > Parent: Construct()
   n5 = Get HideAnimation()
   > PlayAnimation(InAnimation=@n5, StartAtTime=GetEndTime(self=@n5))
   > Set IsShowing(IsShowing=false)
  ]}
 SetFirstLineText {nodes [
   n1 = SetFirstLineText!() -> (then, FirstLineText:Text) {is=title K2Node_FunctionEntry}
   > n2 = Set FirstLineText(FirstLineText:Text=@n1) -> (Output_Get:Text)
   > "SetText (Text)"(self=Get Text_MainText() -> (Text_MainText:Common Text Object Reference), InText=@n2)
  ]}
 SetSecondLineRichText {nodes [
   n1 = SetSecondLineRichText!() -> (then, SecondLineRichText:Text) {is=title K2Node_FunctionEntry}
   > n2 = Set SecondLineRichText(SecondLineRichText:Text=@n1) -> (Output_Get:Text)
   > SetText(self=Get RichText_SecondaryLine(), InText=@n2)
  ]}
]
variables [
 ArrowHorizontalAlignment {type=EHorizontalAlignment, default=HAlign_Right}
 ArrowPadding {type=Margin, default=<ltrb 0 32 0 0>}
 ArrowTransform {type=WidgetTransform, default {angle=90, scale=<1 1>, shear=<0 0>, translation=<20 0>}}
 ArrowVerticalAlignment {type=EVerticalAlignment, default=VAlign_Top}
 FirstLineText {type=string, default="{First Line Text}"}
 IsShowing {type=boolean}
 SecondLineRichText {type=string, default="{Second line text}"}
]
widgets {
 info {inheritedWidgetCount=0, parentClass=&UserWidget, rootWidgetClass=&Overlay, widgetCount=7}
 widgets [
  ParentOverlay {widgetClassPath=&Overlay, widget=ParentOverlay}
  Border_Shadow {
   widgetClassPath=&Border
   bIsVariable=true
   parent=ParentOverlay
   widget=Border_Shadow
   properties {
    verticalAlignment=VAlign_Center
    visibility=SelfHitTestInvisible
    background {
     bIsDynamicallyLoaded=false
     drawAs=Box
     imageType=NoImage
     mirroring=NoMirror
     resourceName=None
     resourceObject=/FortUILibrary/Foundation/Art_ToolKit/AngledBoxes/M_UI_AngledBox_Tuto
     tiling=NoTile
     imageSize=<75 75>
     margin=<ltrb 0.5 0 0.5 0>
     outlineSettings {
      bUseBrushTransparency=false
      roundingType=HalfHeightRadius
      width=0
      color {colorUseRule=UseColor_Specified, specifiedColor=<rgba 0 0 0 0>}
      cornerRadii=<0 0 0 0>
     }
     tintColor {colorUseRule=UseColor_Specified, specifiedColor=<rgba 1 1 1 1>}
     uVRegion {bIsValid=false, max=<0 0>, min=<0 0>}
    }
    brushColor=<rgba 0.1651 0.162 0.1651 1>
    padding=<ltrb 0 0 0 0>
    renderTransform {angle=0, scale=<1 1>, shear=<0 0>, translation=<3 3>}
   }
   slot {class=OverlaySlot, properties {horizontalAlignment=HAlign_Fill, verticalAlignment=VAlign_Fill}}
  }
  Image_Arrow {
   widgetClassPath=&Image
   bIsVariable=true
   parent=ParentOverlay
   widget=Image_Arrow
   properties {
    visibility=SelfHitTestInvisible
    brush {
     bIsDynamicallyLoaded=false
     drawAs=Image
     imageType=NoImage
     mirroring=NoMirror
     resourceName=None
     resourceObject=/Game/Athena/HUD/Quests/Art/T_UI_Quest_Arrow
     tiling=NoTile
     imageSize=<24 24>
     margin=<ltrb 0 0 0 0>
     outlineSettings {
      bUseBrushTransparency=false
      roundingType=HalfHeightRadius
      width=0
      color {colorUseRule=UseColor_Specified, specifiedColor=<rgba 0 0 0 0>}
      cornerRadii=<0 0 0 0>
     }
     tintColor {colorUseRule=UseColor_Specified, specifiedColor=<rgba 1 1 1 1>}
     uVRegion {bIsValid=false, max=<0 0>, min=<0 0>}
    }
    renderTransform {angle=90, scale=<1 1>, shear=<0 0>, translation=<20 0>}
   }
   slot {class=OverlaySlot, properties {horizontalAlignment=HAlign_Right, padding=<ltrb 0 32 0 0>}}
  }
  Border {
   widgetClassPath=&Border
   bIsVariable=true
   parent=ParentOverlay
   widget=Border
   properties {
    verticalAlignment=VAlign_Center
    visibility=SelfHitTestInvisible
    background {
     bIsDynamicallyLoaded=false
     drawAs=Box
     imageType=NoImage
     mirroring=NoMirror
     resourceName=None
     resourceObject=&M_UI_AngledBox_Tuto
     tiling=NoTile
     imageSize=<75 75>
     margin=<ltrb 0.5 0 0.5 0>
     outlineSettings {
      bUseBrushTransparency=false
      roundingType=HalfHeightRadius
      width=0
      color {colorUseRule=UseColor_Specified, specifiedColor=<rgba 0 0 0 0>}
      cornerRadii=<0 0 0 0>
     }
     tintColor {colorUseRule=UseColor_Specified, specifiedColor=<rgba 1 1 1 1>}
     uVRegion {bIsValid=false, max=<0 0>, min=<0 0>}
    }
    padding=<ltrb 0 0 0 0>
   }
   slot {class=OverlaySlot, properties {horizontalAlignment=HAlign_Fill, verticalAlignment=VAlign_Fill}}
  }
  VerticalBox_0 {
   widgetClassPath=&VerticalBox
   parent=ParentOverlay
   widget=VerticalBox_0
   slot {class=OverlaySlot, properties {horizontalAlignment=HAlign_Fill, verticalAlignment=VAlign_Center, padding=<ltrb 48 20 48 20>}}
  }
  Text_MainText {
   widgetClassPath=&CommonTextBlock
   bIsVariable=true
   parent=VerticalBox_0
   widget=Text_MainText
   properties {
    justification=Center
    text="{Main line of text}"
    textTransformPolicy=ToUpper
    wrapTextAt=350
    wrappingPolicy=AllowPerCharacterWrapping
    font {
     bForceMonospaced=false
     bMaterialIsStencil=false
     fontMaterial=null
     fontObject=&Burbank
     letterSpacing=0
     monospacedWidth=1
     size=18
     skewAmount=0
     typefaceFontName=BurbankRegular-Black
     outlineSettings {
      bApplyOutlineToDropShadows=false
      bMiteredCorners=false
      bSeparateFillAlpha=false
      outlineMaterial=null
      outlineSize=0
      outlineColor=<rgba 0 0 0 1>
     }
    }
    renderTransformPivot=<1 0.5>
    shadowOffset=<0 0>
   }
   slot {class=VerticalBoxSlot}
  }
  RichText_SecondaryLine {
   widgetClassPath=&CommonRichTextBlock
   bIsVariable=true
   parent=VerticalBox_0
   widget=RichText_SecondaryLine
   properties {
    bOverrideDefaultStyle=true
    justification=Center
    wrapTextAt=350
    defaultTextStyleOverride {
     overflowPolicy=Clip
     transformPolicy=None
     colorAndOpacity {colorUseRule=UseColor_Specified, specifiedColor=<rgba 0.0314 0.8157 0.9922 1>}
     font {
      bForceMonospaced=false
      bMaterialIsStencil=false
      fontMaterial=null
      fontObject=&BurbankSmall
      letterSpacing=0
      monospacedWidth=1
      size=14
      skewAmount=0
      typefaceFontName=Bold
      outlineSettings {
       bApplyOutlineToDropShadows=false
       bMiteredCorners=false
       bSeparateFillAlpha=false
       outlineMaterial=null
       outlineSize=0
       outlineColor=<rgba 0 0 0 1>
      }
     }
     highlightColor {colorUseRule=UseColor_Specified, specifiedColor=<rgba 0 0 0 1>}
     highlightShape {
      bIsDynamicallyLoaded=false
      drawAs=Image
      imageType=NoImage
      mirroring=NoMirror
      resourceName=None
      resourceObject=null
      tiling=NoTile
      imageSize=<32 32>
      margin=<ltrb 0 0 0 0>
      outlineSettings {
       bUseBrushTransparency=false
       roundingType=HalfHeightRadius
       width=0
       color {colorUseRule=UseColor_Specified, specifiedColor=<rgba 0 0 0 0>}
       cornerRadii=<0 0 0 0>
      }
      tintColor {colorUseRule=UseColor_Specified, specifiedColor=<rgba 1 1 1 1>}
      uVRegion {bIsValid=false, max=<0 0>, min=<0 0>}
     }
     selectedBackgroundColor {colorUseRule=UseColor_Foreground, specifiedColor=<rgba 1 0 1 1>}
     shadowColorAndOpacity=<rgba 0 0 0 1>
     shadowOffset=<0 0>
     strikeBrush {
      bIsDynamicallyLoaded=false
      drawAs=Image
      imageType=NoImage
      mirroring=NoMirror
      resourceName=None
      resourceObject=null
      tiling=NoTile
      imageSize=<32 32>
      margin=<ltrb 0 0 0 0>
      outlineSettings {
       bUseBrushTransparency=false
       roundingType=HalfHeightRadius
       width=0
       color {colorUseRule=UseColor_Specified, specifiedColor=<rgba 0 0 0 0>}
       cornerRadii=<0 0 0 0>
      }
      tintColor {colorUseRule=UseColor_Specified, specifiedColor=<rgba 1 1 1 1>}
      uVRegion {bIsValid=false, max=<0 0>, min=<0 0>}
     }
     underlineBrush {
      bIsDynamicallyLoaded=false
      drawAs=Image
      imageType=NoImage
      mirroring=NoMirror
      resourceName=None
      resourceObject=null
      tiling=NoTile
      imageSize=<32 32>
      margin=<ltrb 0 0 0 0>
      outlineSettings {
       bUseBrushTransparency=false
       roundingType=HalfHeightRadius
       width=0
       color {colorUseRule=UseColor_Specified, specifiedColor=<rgba 0 0 0 0>}
       cornerRadii=<0 0 0 0>
      }
      tintColor {colorUseRule=UseColor_Specified, specifiedColor=<rgba 1 1 1 1>}
      uVRegion {bIsValid=false, max=<0 0>, min=<0 0>}
     }
    }
   }
   slot {class=VerticalBoxSlot}
  }
 ]
}
